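# Physics asset for the AcrylicStitch male medium character body, setting up a hanging garland at the hip.
# Front and back capsule chains hang from zero-size base bodies locked to the pelvis, each branching from its second link into a lower strand.
# Constraints tie front and back chain ends together at both levels, closing two loops with mostly tight angular limits.
# Pelvis and right thigh capsules provide body volumes, with the thigh rotating freely against the pelvis.
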PhysicsAsset /BRCosmetics/Characters/Player/Male/Medium/Bodies/M_MED_AcrylicStitch/Meshes/M_MED_AcrylicStitch_Physics
bodies {
 Base_back_garland [{shapeType=Capsule, length=0, radius=0}]
 Base_front_garland [{shapeType=Capsule, length=0, radius=0}]
 dyn_backBOT_garland_01 [{shapeType=Capsule, length=3.9534, radius=1.8877, center=<0 0 1>, rotation=<rot 73.4339 179.9999 179.9999>}]
 dyn_backBOT_garland_02 [{shapeType=Capsule, length=2.2598, radius=1.9602, center=<-0.2559 0 2>, rotation=<rot 83.0475 179.9999 180>}]
 dyn_backBOT_garland_03 [{shapeType=Capsule, length=1, radius=1}]
 dyn_back_garland_01 [{shapeType=Capsule, length=6.0824, radius=1, center=<1.6 0.6 0>, rotation=<rot 83.3664 179.9997 179.9997>}]
 dyn_back_garland_02 [{shapeType=Capsule, length=4.6639, radius=1.5068, center=<1.5 0 1>, rotation=<rot 83.3664 179.9998 179.9998>}]
 dyn_back_garland_03 [{shapeType=Capsule, length=5.29, radius=1.5338, center=<1.5 0 1.4>, rotation=<rot 87.9846 -0.0001 0.0001>}]
 dyn_back_garland_04 [{shapeType=Capsule, length=3.2532, radius=1.9535, center=<0.5 0 1>, rotation=<rot 64.1346 -0.0065 0.0003>}]
 dyn_frontBOT_garland_01 [{shapeType=Capsule, length=2.5404, radius=2.6063, center=<0.051 0 1.0004>, rotation=<rot 89.5728 179.9918 179.9918>}]
 dyn_frontBOT_garland_02 [{shapeType=Capsule, length=2.3722, radius=2.7745, center=<0.0501 0 0.8>, rotation=<rot 85.7421 179.9977 179.998>}]
 dyn_frontBOT_garland_03 [{shapeType=Capsule, length=3.8103, radius=2.8094, center=<0 0 2.2>, rotation=<rot 74.5943 -8.6172 -24.2861>}]
 dyn_front_garland_01 [{shapeType=Capsule, length=2.5247, radius=1.7687, center=<1.389 0 0>, rotation=<rot 82.872 180 180>}]
 dyn_front_garland_02 [{shapeType=Capsule, length=4.6064, radius=1.9065, center=<1.4 0 0.7>, rotation=<rot 90 0 0>}]
 dyn_front_garland_03 [{shapeType=Capsule, length=4.0212, radius=2.3471, center=<1 0 1.2>, rotation=<rot 75.3986 179.9996 179.9997>}]
 dyn_front_garland_04 [{shapeType=Capsule, length=4.2497, radius=2.2397, center=<0 0 1>, rotation=<rot 77.7572 17.7646 0.0001>}]
 pelvis [{shapeType=Capsule, length=12.2751, radius=8.3349, center=<-2.6363 0 0>}]
 thigh_twist_01_r [
  {
   shapeType=Capsule
   length=17.2366
   radius=8.2972
   shapeName=thigh_twist_01_r_capsule
   center=<-8.0774 0.2561 0.0193>
   rotation=<rot 20.829 93.9975 -90>
  }
  {
   shapeType=Capsule
   length=15
   radius=10.3371
   shapeName=thigh_twist_01_r_capsule
   center=<-9.6359 0 -8.8418>
   rotation=<rot 80.5411 179.9999 179.9999>
  }
 ]
}
constraints [
 {
  bone1Name=thigh_twist_01_r
  bone2Name=pelvis
  swing1LimitDegrees=45
  swing1Motion=Free
  swing2LimitDegrees=45
  swing2Motion=Free
  twistLimitDegrees=45
  twistMotion=Free
 }
 {
  bone1Name=Base_back_garland
  bone2Name=pelvis
  swing1LimitDegrees=45
  swing1Motion=Locked
  swing2LimitDegrees=45
  swing2Motion=Locked
  twistLimitDegrees=45
  twistMotion=Locked
 }
 {bone1Name=dyn_back_garland_01, bone2Name=Base_back_garland, swing1LimitDegrees=30, swing2LimitDegrees=20, twistLimitDegrees=0}
 {bone1Name=dyn_back_garland_02, bone2Name=dyn_back_garland_01, swing1LimitDegrees=30, swing2LimitDegrees=20, twistLimitDegrees=0}
 {bone1Name=dyn_back_garland_03, bone2Name=dyn_back_garland_02, swing1LimitDegrees=30, swing2LimitDegrees=20, twistLimitDegrees=1}
 {
  bone1Name=dyn_back_garland_04
  bone2Name=dyn_back_garland_03
  swing1LimitDegrees=45
  swing1Motion=Locked
  swing2LimitDegrees=45
  swing2Motion=Locked
  twistLimitDegrees=45
  twistMotion=Locked
 }
 {
  bone1Name=Base_front_garland
  bone2Name=pelvis
  swing1LimitDegrees=45
  swing1Motion=Locked
  swing2LimitDegrees=45
  swing2Motion=Locked
  twistLimitDegrees=45
  twistMotion=Locked
 }
 {bone1Name=dyn_front_garland_01, bone2Name=Base_front_garland, swing1LimitDegrees=30, swing2LimitDegrees=20, twistLimitDegrees=0}
 {bone1Name=dyn_front_garland_02, bone2Name=dyn_front_garland_01, swing1LimitDegrees=30, swing2LimitDegrees=20, twistLimitDegrees=0}
 {bone1Name=dyn_front_garland_03, bone2Name=dyn_front_garland_02, swing1LimitDegrees=30, swing2LimitDegrees=20, twistLimitDegrees=1}
 {bone1Name=dyn_front_garland_04, bone2Name=dyn_front_garland_03, swing1LimitDegrees=5, swing2LimitDegrees=5, twistLimitDegrees=0}
 {bone1Name=dyn_front_garland_04, bone2Name=dyn_back_garland_04, swing1LimitDegrees=1, swing2LimitDegrees=1, twistLimitDegrees=1}
 {bone1Name=dyn_backBOT_garland_03, bone2Name=dyn_backBOT_garland_02, swing1LimitDegrees=1, swing2LimitDegrees=1, twistLimitDegrees=5}
 {bone1Name=dyn_backBOT_garland_02, bone2Name=dyn_backBOT_garland_01, swing1LimitDegrees=10, swing2LimitDegrees=10, twistLimitDegrees=5}
 {bone1Name=dyn_backBOT_garland_01, bone2Name=dyn_back_garland_02, swing1LimitDegrees=20, swing2LimitDegrees=20, twistLimitDegrees=5}
 {bone1Name=dyn_frontBOT_garland_03, bone2Name=dyn_frontBOT_garland_02, swing1LimitDegrees=1, swing2LimitDegrees=1, twistLimitDegrees=5}
 {bone1Name=dyn_frontBOT_garland_02, bone2Name=dyn_frontBOT_garland_01, swing1LimitDegrees=10, swing2LimitDegrees=10, twistLimitDegrees=5}
 {bone1Name=dyn_frontBOT_garland_01, bone2Name=dyn_front_garland_02, swing1LimitDegrees=20, swing2LimitDegrees=20, twistLimitDegrees=5}
 {bone1Name=dyn_frontBOT_garland_03, bone2Name=dyn_backBOT_garland_03, swing1LimitDegrees=1, swing2LimitDegrees=1, twistLimitDegrees=1}
]
