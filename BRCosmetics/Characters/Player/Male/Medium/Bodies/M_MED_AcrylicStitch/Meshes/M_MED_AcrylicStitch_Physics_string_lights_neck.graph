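# Physics asset for the neck string-lights garland on the M_MED_AcrylicStitch male medium player body, covering the garland bones and upper torso.
# Two box bodies, dyn_garland_01 and dyn_garland_02, chain from a zero-radius Base_garland sphere, with small swing and twist limit angles.
# Base_garland and both clavicle capsules are angularly locked to the spine_05 capsule, forming a rigid upper-torso shape around the garland.
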PhysicsAsset /BRCosmetics/Characters/Player/Male/Medium/Bodies/M_MED_AcrylicStitch/Meshes/M_MED_AcrylicStitch_Physics_string_lights_neck
bodies {
 Base_garland [{shapeType=Sphere, radius=0}]
 clavicle_l [{shapeType=Capsule, length=12.7264, radius=8.874, center=<6.8439 -1.3354 -0.0156>, rotation=<rot -89.8695 179.9497 -179.9592>}]
 clavicle_r [{shapeType=Capsule, length=12.7264, radius=8.874, center=<-6.8439 1.3354 0.0156>, rotation=<rot -89.8695 179.9497 0.0408>}]
 dyn_garland_01 [{shapeType=Box, extentX=5.538, extentY=15.5156, extentZ=2.5399, shapeName=dyn_garland_01_box, center=<0 0 1.2>, rotation=<rot 2.2052 0 0>}]
 dyn_garland_02 [{shapeType=Box, extentX=7.8763, extentY=14.7349, extentZ=1.7832, center=<0 0 1.1>, rotation=<rot -6.8309 0 0>}]
 spine_05 [{shapeType=Capsule, length=17.481, radius=9.8001, center=<0.5519 -0.7748 0>}]
}
constraints [
 {bone1Name=dyn_garland_01, bone2Name=Base_garland, swing1LimitDegrees=0, swing2LimitDegrees=20, twistLimitDegrees=20}
 {bone1Name=dyn_garland_02, bone2Name=dyn_garland_01, swing1LimitDegrees=0, swing2LimitDegrees=25, twistLimitDegrees=10}
 {
  bone1Name=clavicle_l
  bone2Name=spine_05
  swing1LimitDegrees=45
  swing1Motion=Locked
  swing2LimitDegrees=45
  swing2Motion=Locked
  twistLimitDegrees=45
  twistMotion=Locked
 }
 {
  bone1Name=clavicle_r
  bone2Name=spine_05
  swing1LimitDegrees=45
  swing1Motion=Locked
  swing2LimitDegrees=45
  swing2Motion=Locked
  twistLimitDegrees=45
  twistMotion=Locked
 }
 {
  bone1Name=Base_garland
  bone2Name=spine_05
  swing1LimitDegrees=45
  swing1Motion=Locked
  swing2LimitDegrees=45
  swing2Motion=Locked
  twistLimitDegrees=45
  twistMotion=Locked
 }
]
